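# Stereo MetaSound source for the rotation sound of the vehicle machine gun turret mod, with concurrency limited to one instance per owner.
# Two looping wave players start on play and mix, one gain driven by vertical aim delta and the other by horizontal, each range-mapped and interpolated.
# An attack-release envelope shapes the mix; OnEnd triggers the release, and envelope completion fires OnFinished.
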MetaSoundSource /VehicleMods/Mods/MachineGunMod/Audio/MSS_Turret_VehMG_Rotate
inputs [
 Attack Curve {type=Float}
 Attack Time {type=Time}
 HorizontalAimDelta {type=Float}
 OnEnd {type=Trigger}
 Release Curve {type=Float}
 Release Time {type=Time}
 UE.Source.OnPlay {type=Trigger}
 VerticalAimDelta {type=Float}
]
interfaces [UE.OutputFormat.Stereo:1.0, UE.Source.OneShot:1.0, UE.Source:1.1]
outputs [
 "UE.OutputFormat.Stereo.Audio:0" {type=Audio}
 "UE.OutputFormat.Stereo.Audio:1" {type=Audio}
 UE.Source.OneShot.OnFinished {type=Trigger}
]
pages [Default {
  input_defaults {
   Attack Curve=1
   Attack Time=0.25
   HorizontalAimDelta=0.38
   OnEnd=null
   Release Curve=1
   Release Time=0.25
   UE.Source.OnPlay=null
   VerticalAimDelta=0
  }
  nodes [
   ADSREnvelopeADSREnvelope_0 = ADSR Envelope.ADSR Envelope:Audio(
    Attack Curve=^Attack Curve
    Attack Time=^Attack Time
    Trigger Release=^OnEnd
    Release Curve=^Release Curve
    Release Time=^Release Time
    Trigger Attack=^UE.Source.OnPlay
   ) {also {defaults {Decay Time=0, Sustain Level=1}}}
   WavePlayer_0 = UE.Wave Player:Stereo(Play=^UE.Source.OnPlay) {also {defaults {Loop="<MetaSoundBuilderResult.SUCCEEDED: 0>", Wave Asset=&/Turret_VehMG_Rotate_Loop_01}}}
   WavePlayer_1 = UE.Wave Player:Stereo(Play=^UE.Source.OnPlay) {also {defaults {Loop="<MetaSoundBuilderResult.SUCCEEDED: 0>", Wave Asset=/Game/Sounds/Fort_Weapons/Turret/WPN_Turret_Mounted_Heavy_Turn_Loop_01}}}
   AudioMixerAudioMixerStereo2_0 = "AudioMixer.Audio Mixer (Stereo, 2)"(
    Gain 0=UE.InterpTo:Audio(Target=MapRange.MapRange:Float(In=^VerticalAimDelta) {also {defaults {In Range B=12.5}}}.Out Value) {also {defaults {Interp Time=0.35}}}.Value
    Gain 1=UE.InterpTo:Audio(Target=MapRange.MapRange:Float(In=^HorizontalAimDelta) {also {defaults {In Range B=12.5}}}.Out Value) {also {defaults {Interp Time=0.35}}}.Value
    In 0 L=@WavePlayer_0.Out Left
    In 0 R=@WavePlayer_0.Out Right
    In 1 L=@WavePlayer_1.Out Left
    In 1 R=@WavePlayer_1.Out Right
   )
   Multiply_0 = UE.Multiply:Audio(PrimaryOperand=@ADSREnvelopeADSREnvelope_0.Out Envelope, AdditionalOperands=@AudioMixerAudioMixerStereo2_0.Out L)
   Multiply_1 = UE.Multiply:Audio(PrimaryOperand=@ADSREnvelopeADSREnvelope_0.Out Envelope, AdditionalOperands=@AudioMixerAudioMixerStereo2_0.Out R)
  ]
  links [
   {dst="@UE.Source.OneShot.OnFinished", src=ADSREnvelopeADSREnvelope_0.On Done}
   {dst="@UE.OutputFormat.Stereo.Audio:0", src=Multiply_0.Out}
   {dst="@UE.OutputFormat.Stereo.Audio:1", src=Multiply_1.Out}
  ]
 }]
properties {
 assetImportData=&self:AssetImportData
 attenuationSettings=&/ATT_Turret_VehMod_Foley
 bOverrideConcurrency=true
 maxDistance=2250
 numChannels=2
 outputFormat=Stereo
 sampleRate=48000
 soundClassObject=&Fort_Traps
 volume=0.7
 concurrencyOverrides {bLimitToOwner=true, maxCount=1}
}
