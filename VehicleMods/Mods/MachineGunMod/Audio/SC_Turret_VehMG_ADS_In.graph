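# Sound cue for the foley played when aiming down sights on the vehicle machine gun mod turret.
# A Random node picks one of three equally weighted aim-in wave variations, then feeds a Modulator at the root.
# The cue plays at half volume in the weapon foley sound class, using the vehicle mod turret foley attenuation.
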SoundCue /VehicleMods/Mods/MachineGunMod/Audio/SC_Turret_VehMG_ADS_In
root=Modulator_0
nodes [
 {id=Modulator_0, children [Random_0]}
 {id=Random_0, children [WavePlayer_0,WavePlayer_2,WavePlayer_3], properties {weights [1,1,1]}}
 {id=WavePlayer_0, properties {soundWaveAssetPtr=&/Turret_VehMG_Ads_In_01}}
 {id=WavePlayer_2, properties {soundWaveAssetPtr=&/Turret_VehMG_Ads_In_02}}
 {id=WavePlayer_3, properties {soundWaveAssetPtr=&/Turret_VehMG_Ads_In_03}}
]
properties {
 attenuationSettings=&/ATT_Turret_VehMod_Foley
 maxDistance=2250
 soundClassObject=&Fort_WeaponFoley
 volumeMultiplier=0.5
}
