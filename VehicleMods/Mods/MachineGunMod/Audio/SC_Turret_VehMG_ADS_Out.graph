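# Sound cue for the foley heard when leaving aim-down-sights on the vehicle mod machine gun turret.
# A Random node picks one of three equally weighted "Ads Out" wave variations and feeds a Modulator left at default settings.
# The cue plays at reduced volume in the weapon foley sound class, spatialized through the shared vehicle mod turret foley attenuation.
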SoundCue /VehicleMods/Mods/MachineGunMod/Audio/SC_Turret_VehMG_ADS_Out
root=Modulator_0
nodes [
 {id=Modulator_0, children [Random_1]}
 {id=Random_1, children [WavePlayer_4,WavePlayer_5,WavePlayer_6], properties {weights [1,1,1]}}
 {id=WavePlayer_4, properties {soundWaveAssetPtr=&/Turret_VehMG_Ads_Out_01}}
 {id=WavePlayer_5, properties {soundWaveAssetPtr=&/Turret_VehMG_Ads_Out_02}}
 {id=WavePlayer_6, properties {soundWaveAssetPtr=&/Turret_VehMG_Ads_Out_03}}
]
properties {
 attenuationSettings=&/ATT_Turret_VehMod_Foley
 maxDistance=2250
 soundClassObject=&Fort_WeaponFoley
 volumeMultiplier=0.5
}
