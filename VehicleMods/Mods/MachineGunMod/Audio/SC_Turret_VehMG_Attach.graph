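# Sound cue for attaching the machine gun turret vehicle mod, classed as weapon foley.
# A single wave player feeds the root directly and plays the Turret_VehMG_Engage_01 wave, with no randomization or mixing nodes.
# A shared vehicle mod foley attenuation asset spatializes the sound and limits how far it carries.
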SoundCue /VehicleMods/Mods/MachineGunMod/Audio/SC_Turret_VehMG_Attach
root=WavePlayer_0
nodes [{id=WavePlayer_0, properties {soundWaveAssetPtr=&/Turret_VehMG_Engage_01}}]
properties {
 attenuationSettings=&/ATT_Turret_VehMod_Foley
 maxDistance=2250
 soundClassObject=&Fort_WeaponFoley
 volumeMultiplier=1
}
