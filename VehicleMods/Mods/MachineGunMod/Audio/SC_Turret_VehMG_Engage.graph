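# Sound cue for the engage sound of the vehicle machine gun turret mod, classed as weapon foley.
# A single wave player is the root and plays the turret's equip sound wave, with no randomization, mixing or modulation nodes.
# The vehicle mod foley attenuation asset governs its spatial falloff.
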SoundCue /VehicleMods/Mods/MachineGunMod/Audio/SC_Turret_VehMG_Engage
root=WavePlayer_1
nodes [{id=WavePlayer_1, properties {soundWaveAssetPtr=&/Turret_VehMG_Equip_01}}]
properties {
 attenuationSettings=&/ATT_Turret_VehMod_Foley
 maxDistance=2250
 soundClassObject=&Fort_WeaponFoley
 volumeMultiplier=1
}
