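# Sound cue for the looping overheat sound of the vehicle-mounted machine gun turret mod.
# A looping wave player feeds a continuous modulator, which scales both pitch and volume from the Overheat parameter so the loop tracks the gun's heat.
# The cue uses the vehicle mod turret foley attenuation and its own concurrency override that prevents new instances once the limit is reached.
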SoundCue /VehicleMods/Mods/MachineGunMod/Audio/Turret_VehMG_OverheatLoop_Cue
root=ModulatorContinuous_1
nodes [
 {
  id=ModulatorContinuous_1
  children [WavePlayer_0]
  properties {
   pitchModulationParams {default=0, maxOutput=1.3, minOutput=0.95, parameterName=Overheat}
   volumeModulationParams {default=0, parameterName=Overheat}
  }
 }
 {id=WavePlayer_0, properties {bLooping=true, soundWaveAssetPtr=/Game/Sounds/Fort_Weapons/Turret/WPN_Turret_Mounted_Auto_Overheat_Loop_01}}
]
properties {
 attenuationSettings=&/ATT_Turret_VehMod_Foley
 bOverrideConcurrency=true
 maxDistance=2250
 soundClassObject=/Game/Sounds/FortSoundClassesAndModes/Fort_WeaponRanged_LowAmmo_Critical_PSM
 volumeMultiplier=0.8
 attenuationOverrides {distanceAlgorithm=NaturalSound, falloffDistance=4000, attenuationShapeExtents {x=1000}}
 concurrencyOverrides {resolutionRule=PreventNew}
}
